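; Blueprint for a glowing Dark Village light fixture prop in the Apollo environment building actors.
; Its construction script only calls the parent function, so setup comes from the parent class.
; An EventCallAnimationCurveAgain event is declared with no logic attached.
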(event EventCallAnimationCurveAgain)

(fn ConstructionScript
  (|CallParentFunction))
